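# Sound cue for a bamboo village door closing, routed to the Fort_Doors sound class.
# A Random node picks one of four equally weighted close samples, which passes through a Modulator into a Quality Level root with only its first input connected.
# It uses shared door attenuation at reduced volume, with a concurrency override that prevents new instances once the limit is reached.
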SoundCue /Game/Sounds/Asteria/Environment_Actors/Door_Bamboo/Door_Wood_Bamboo_Village_Close_Cue
root=QualityLevel_0
nodes [
 {id=Modulator_0, children [Random_2]}
 {id=QualityLevel_0, children [Modulator_0,null,null,null]}
 {id=Random_2, children [WavePlayer_4,WavePlayer_5,WavePlayer_7,WavePlayer_8], properties {weights [1,1,1,1]}}
 {id=WavePlayer_4, properties {soundWaveAssetPtr=&/Bamboo_Door_Close_01}}
 {id=WavePlayer_5, properties {soundWaveAssetPtr=&/Bamboo_Door_Close_02}}
 {id=WavePlayer_7, properties {soundWaveAssetPtr=&/Bamboo_Door_Close_03}}
 {id=WavePlayer_8, properties {soundWaveAssetPtr=&/Bamboo_Door_Close_04}}
]
properties {
 attenuationSettings=/Game/Sounds/Attenuation/Fort_Doors_Att
 bOverrideConcurrency=true
 maxDistance=2400
 soundClassObject=&Fort_Doors
 volumeMultiplier=0.6
 attenuationOverrides {falloffDistance=1600}
 concurrencyOverrides {resolutionRule=PreventNew}
}
