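# Sound cue for a bamboo village door opening, in the Fort_Doors sound class.
# A modulator wraps an equally weighted random pick of three waves: two bamboo door open sounds and, despite the cue's name, one close sound.
# It uses shared fort door attenuation at reduced volume, and its own concurrency override prevents new instances when the limit is reached.
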SoundCue /Game/Sounds/Asteria/Environment_Actors/Door_Bamboo/Door_Wood_Bamboo_Village_Open_Cue
root=Modulator_1
nodes [
 {id=Modulator_1, children [Random_2]}
 {id=Random_2, children [WavePlayer_7,WavePlayer_8,WavePlayer_9], properties {weights [1,1,1]}}
 {id=WavePlayer_7, properties {soundWaveAssetPtr=&/Bamboo_Door_Close_02}}
 {id=WavePlayer_8, properties {soundWaveAssetPtr=&/Bamboo_Door_Open_02}}
 {id=WavePlayer_9, properties {soundWaveAssetPtr=&/Bamboo_Door_Open_03}}
]
properties {
 attenuationSettings=/Game/Sounds/Attenuation/Fort_Doors_Att
 bOverrideConcurrency=true
 maxDistance=2400
 soundClassObject=&Fort_Doors
 volumeMultiplier=0.6
 attenuationOverrides {falloffDistance=1600}
 concurrencyOverrides {resolutionRule=PreventNew}
}
